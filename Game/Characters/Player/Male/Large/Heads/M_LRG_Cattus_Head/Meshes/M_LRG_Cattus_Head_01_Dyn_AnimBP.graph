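# Animation Blueprint for the large male Cattus character head, adding secondary jaw motion and pose adjustments for targeting and roaring.
# While targeting, it blends from the cached input pose to one with additive head and dyn_hairBase rotations.
# A low-alpha spring Anim Dynamics node then moves C_jaw.
# The update event detects the CattusRoar emote montage and raises a timed alpha, interpolated to drive an extra dyn_hairBase rotation.
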AnimBlueprint /Game/Characters/Player/Male/Large/Heads/M_LRG_Cattus_Head/Meshes/M_LRG_Cattus_Head_01_Dyn_AnimBP
target_skeleton=/Game/Characters/Player/Common/Fortnite_Base_Head/Fortnite_Base_Head_Export_Skeleton
graphs [
 AnimGraph {nodes [
   n1 = "Transform (Modify) Bone - Bone: head"!(ComponentPose=Local To Component(LocalPose=Use cached pose 'SGI Cache'()), Alpha=1, Rotation="0,-30,0") -> (Pose) {rotationMode=BMM_Additive, scaleMode=BMM_Additive, alphaBoolBlend {blendOption=Linear}, boneToModify {boneName=head}, rotation {yaw=-30}}
   n2 = "Transform (Modify) Bone - Bone: dyn_hairBase"!(ComponentPose=@n1, Alpha=1, Rotation="0,18,0") -> (Pose) {
    rotationMode=BMM_Additive
    scaleMode=BMM_Additive
    alphaBoolBlend {blendOption=Linear}
    boneToModify {boneName=dyn_hairBase}
    rotation {yaw=18}
   }
   n3 = Component To Local(ComponentPose=@n2)
   n4 = Use cached pose 'SGI Cache'()
   n5 = Get bIsTargeting()
   n6 = Anim Dynamics - Bone: C_jaw(
    ComponentPose=Local To Component(LocalPose=Blend Poses by bool(BlendPose_0=@n3, BlendPose_1=@n4, BlendTime_0=0.15, bActiveValue=@n5) {blendType=Linear})
    Alpha=0.12
    ExternalForce="0,0,0"
   ) {
    angularDampingOverride=1
    angularSpringConstant=1500
    bAngularSpring=true
    bGravityOverrideInSimSpace=true
    bOverrideAngularDamping=true
    gravityScale=0.35
    simulationSpace=RootRelative
    alphaBoolBlend {blendOption=Linear}
    boundBone {boneName=C_jaw}
    componentLinearAccScale {z=0.46}
    componentLinearVelScale {z=0.46}
    physicsBodyDefinitions [{
      collisionType=CustomSphere
      sphereCollisionRadius=2
      boundBone {boneName=C_jaw}
      boxExtents=<20 20 10>
      constraintSetup {angularTargetAxis=AxisZ, twistAxis=AxisZ, angularLimitsMax=<80 10 0>, angularLimitsMin=<-80 -10 0>, angularTarget=<0 0 0>}
      localJointOffset=<0 0 -15>
     }]
   }
   n7 = "Transform (Modify) Bone - Bone: dyn_hairBase"(ComponentPose=@n6, Alpha=Get isRoaringAlpha() -> (isRoaringAlpha:double), Rotation="0,15,0") {
    rotationMode=BMM_Additive
    rotationSpace=BCS_BoneSpace
    scaleMode=BMM_Additive
    alphaBoolBlend {blendOption=Linear}
    alphaScaleBiasClamp {bInterpResult=true, interpSpeedDecreasing=3, interpSpeedIncreasing=3}
    boneToModify {boneName=dyn_hairBase}
    rotation {yaw=15}
   }
   Output Pose(Result=Component To Local(ComponentPose=@n7)) {layerGroup=None}
   Save cached pose 'SGI Cache'(Pose=Input Pose()) {cacheName=SGI Cache}
  ]}
 EventGraph {nodes [
   Event BlueprintUpdateAnimation()
   n1 = GetAnimInstance(self=GetOwnerSkeletalMeshComponent())
   > Is Valid(InputObject=@n1)
   > n2 = Branch(Condition=Montage_IsPlaying(self=@n1, Montage=/Game/Animation/Game/MainPlayer/Emotes/CattusRoar/Emote_CattusRoar_CMM_M))
   > Do Once(Reset=@n3.then)
   > Set isRoaringAlpha(isRoaringAlpha:double=1) -> (Output_Get:double)
   > Delay(Duration=4)
   > Set isRoaringAlpha(isRoaringAlpha=0)
   n3 = Set isRoaringAlpha(execute=@n2.else, isRoaringAlpha=0)
   TryGetPawnOwner()
  ]}
]
variables [isRoaringAlpha {type=number}]
